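; Child Blueprint of a generic timer manager that adds countdown tick audio to the game's timer.
; On BeginPlay, only where cosmetic events can execute, it binds to OnSecondsRemainingUpdated.
; Each non-negative update calls Play on the audio component, passes SecondsRemaining as an integer parameter and fires the InClockTick trigger, leaving sound variation to the audio asset.
; EndPlay unbinds the handler and fades the audio out.
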(event EventBeginPlay
  (|CallParentFunction)
  (Networking|CanExecuteCosmeticEvents
    (:True
      (Default|BindEventtoOnSecondsRemainingUpdated (delegate Custom|Play%20Tick%20Sound)))))

(event Custom|Play%20Tick%20Sound
  (params (SecondsRemaining :type Integer))
  (if (>= SecondsRemaining 0)
    (bind _audiocomponent (Variables|Default|GetAudioComponent)
      (:then
        (Audio|Components|Audio|Play _audiocomponent 0.0)
        (Audio|Parameter|SetIntegerParameter _audiocomponent "SecondsRemaining" SecondsRemaining :!class /Script/AudioExtensions.AudioParameterControllerInterface)
        (Audio|Parameter|ExecuteTriggerParameter _audiocomponent "InClockTick"))
      (:else))))

(event EventEndPlay
  (params (EndPlayReason :type EEndPlayReason%20Enum :default "Destroyed"))
  (Networking|CanExecuteCosmeticEvents
    (:True
      (Default|UnbindEventfromOnSecondsRemainingUpdated (delegate Custom|Play%20Tick%20Sound))
      (bind _audiocomponent (Variables|Default|GetAudioComponent)
        (:then
          (Audio|Components|Audio|FadeOut _audiocomponent 3.0 0.0 "Linear" :!class /Script/Engine.AudioComponent))
        (:else)))))

(fn ConstructionScript
  (|CallParentFunction))
