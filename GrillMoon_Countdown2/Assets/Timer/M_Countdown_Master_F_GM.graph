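# Unlit translucent master material for a holographic world-space countdown timer with digits and DHMS labels, crossfading to a post-countdown image in phase 3.
# World position offset yaws the mesh toward the camera, adds optional noise displacement, and sinks it out of sight when MPC_Countdown flags or distance fade hide it.
# Panning world-aligned noise, sine-driven flicker and distance-scaled scanlines shape opacity.
# A curve atlas tints the emissive.
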Material /GrillMoon_Countdown2/Assets/Timer/M_Countdown_Master_F_GM
graph {
 properties {
  bEnableResponsiveAA=true
  bUsedWithInstancedStaticMeshes=true
  blendMode=BLEND_Translucent
  pixelDepthOffsetMode=PDOM_Legacy
  shadingModel=MSM_Unlit
  usedShadingModels=Unlit
 }
 nodes [
  n1 = ObjectPositionWS()
  n2 = WorldPosition()
  n3 = CameraPositionWS()
  n4 = ActorPositionWS()
  n5 = Normalize(VectorInput=Multiply(A=Constant3Vector() {constant=<rgba 1 1 0 0>}, B=Subtract(A=@n3, B=@n4)))
  n6 = MaterialFunctionCall!(Normalized V1=@n5, Normalized V2=Normalize(VectorInput=Constant3Vector() {constant=<rgba 0 1 0 0>})) {materialFunction=/Engine/Functions/Engine_MaterialFunctions02/Math/AngleBetweenVectors}
  n7 = Constant3Vector() {constant=<rgba 1 0 0 0>}
  n8 = Sign(None=DotProduct(A=@n5, B=Normalize(VectorInput=@n7)))
  n9 = RotateAboutAxis(
   NormalizedRotationAxis=Constant3Vector() {constant=<rgba 0 0 1 0>}
   RotationAngle=Multiply(A=Divide(A=@n6.Angle In Degrees) {constB=360}, B=Multiply(A=@n8) {constB=-1})
   PivotPoint=@n4
   Position=WorldPosition().XYZ
  )
  n10 = ScalarParameter() {defaultValue=1, parameterName=CameraFacing}
  n11 = TextureObject() {samplerType=SAMPLERTYPE_LinearColor, texture=&T_NoisePack_LinearColor}
  n12 = ScalarParameter() {defaultValue=1500, group=Noise, parameterName=NoiseSize}
  n13 = Multiply(A=VectorParameter() {group=Noise, parameterName=NoiseTileMult, defaultValue=<rgba 2 2 0.1 1>}.RGB, B=@n12)
  n14 = Time()
  n15 = VectorParameter() {group=Noise, parameterName=NoisePanSpeed, defaultValue=<rgba -0.1 0.09 -0.15 0>}
  n16 = Add(A=WorldPosition().XYZ, B=Multiply(A=@n12, B=Multiply(A=@n14, B=@n15.RGB)))
  n17 = Multiply(
   A=Power(Base=MaterialFunctionCall(TextureObject=@n11, TextureSize=@n13, WorldPosition=@n16) {materialFunction=&WorldAlignedTexture}.XY Texture) {constExponent=4}
   B=ScalarParameter() {defaultValue=2, group=Noise, parameterName=NoiseMaskMult}
  ) {constB=20}
  n18 = Time()
  n19 = Multiply(A=ScalarParameter() {defaultValue=1, parameterName=FlickerSpeed}, B=@n18)
  n20 = Sine(None=@n19) {period=1.6}
  n21 = Sine(None=@n19) {period=1.42}
  n22 = Ceil(None=AppendVector(A=AppendVector(A=@n20, B=@n21), B=Sine(None=@n19) {period=1.51}))
  n23 = ScalarParameter() {defaultValue=0.5, group=Noise, parameterName=NoiseFrequency}
  n24 = Add(A=Sine(None=@n19) {period=0.72}, B=Multiply(B=@n23) {constA=0.75}) {constB=0.75}
  n25 = Multiply(A=Add(A=Sine(None=@n19) {period=2.1}, B=Multiply(B=@n23) {constA=0.25}), B=@n24)
  n26 = Multiply(B=@n23) {constA=0.9, constB=0.9}
  n27 = Sine(None=@n19) {period=1.25}
  n28 = Multiply(A=@n24, B=Ceil(None=Add(A=@n26, B=@n27) {constB=0.9}))
  n29 = Sine(None=@n18) {period=0.075}
  n30 = Multiply(A=Saturate(None=Multiply(A=@n25, B=@n28)), B=LinearInterpolate(Alpha=ConstantBiasScale(None=@n29)) {constA=0.7})
  n31 = Add(
   A=ScalarParameter() {defaultValue=0.15, group=Noise, parameterName=NoiseAdd}
   B=Multiply(A=Multiply(A=@n22, B=@n30), B=ScalarParameter() {defaultValue=2, group=Noise, parameterName=NoiseMult})
  )
  n32 = Desaturation(None=Saturate(None=Multiply(A=@n17, B=@n31)))
  n33 = LinearInterpolate(A=@n32, B=Constant(), Alpha=ScalarParameter() {defaultValue=1, group=Noise, parameterName=RemoveNoisetoWPO, sortPriority=0})
  n34 = Subtract(
   A=WorldPosition().XYZ
   B=MaterialFunctionCall() {materialFunction=&ObjectPivotPoint}.Object Pivot Location
  )
  n35 = Multiply(
   A=Multiply(A=Normalize(VectorInput=@n34), B=Constant3Vector() {constant=<rgba 1 1 0 0>})
   B=ScalarParameter() {defaultValue=100, group=Noise, parameterName=WPOStrength}
  )
  n36 = OneMinus(None=MaterialFunctionCall(
    Fade Length=ScalarParameter() {defaultValue=45000, parameterName=DistanceFadeLength}
    Fade Offset=ScalarParameter() {defaultValue=40000, parameterName=DistanceFadeOffset}
    For Vertex Shader=StaticBool() {value=true}
   ) {materialFunction=&CameraDepthFade}.Result)
  n37 = Ceil(None=@n36)
  n38 = Constant() {r=1}
  n39 = CollectionParameter() {collection=/Game/Athena/Events/Generic/CountdownTimer/Materials/MPC_Countdown, parameterName=ShowCountdown}
  n40 = StaticSwitchParameter(
   True=Multiply(
    A=@n39
    B=CollectionParameter() {collection=&MPC_Countdown, parameterName=ShowCountdownPhase2}
   )
   False=@n39
  ) {parameterName=isPhase2}
  LinearInterpolate(
   A=Add(A=Subtract(A=@n1, B=@n2.XYZ), B=Constant3Vector() {constant=<rgba 0 0 -5156156 0>})
   B=Add(A=Multiply(A=@n33, B=@n35) {constB=0}, B=LinearInterpolate(B=@n9, Alpha=@n10))
   Alpha=Multiply(A=Saturate(None=@n37), B=StaticSwitchParameter(True=@n38, False=@n40) {parameterName=Debug})
  )
  n41 = Constant() {r=0.5}
  n42 = SphereMask(A=TextureCoordinate(), B=@n41, Radius=@n41) {hardnessPercent=0}
  n43 = ScalarParameter() {group=Numbers, parameterName=u_DHMSPos}
  n44 = ScalarParameter() {defaultValue=-0.6, group=Numbers, parameterName=v_DHMSPos}
  n45 = ScalarParameter() {group=Numbers, parameterName=u_NumberPaddingOffset}
  n46 = ScalarParameter() {group=Numbers, parameterName=v_NumberPaddingOffset}
  n47 = MaterialFunctionCall!(
   UVs=Add(A=TextureCoordinate(), B=AppendVector(A=@n45, B=@n46))
   UVPadding=AppendVector(
    A=ScalarParameter() {defaultValue=1, group=Numbers, parameterName=u_NumberPaddingScale}
    B=ScalarParameter() {group=Numbers, parameterName=v_NumberPaddingScale}
   )
   "NumberofDigits(int)"=ScalarParameter() {defaultValue=11, parameterName=NumberofDigits}
  ) {materialFunction=/Game/Packages/Fortress_SharedMaterials/Base_Material_Functions/BaseCharacterShaderFunctions/MF_CountdownTimer}
  n48 = MaterialFunctionCall(
   UVs=Add(A=AppendVector(A=@n43, B=@n44), B=@n47.UVS)
   Texture Scale=AppendVector(
    A=ScalarParameter() {defaultValue=1.0383, group=Numbers, parameterName=u_DHMSScale}
    B=ScalarParameter() {defaultValue=0.419, group=Numbers, parameterName=v_DHMSScale}
   )
  ) {materialFunction=&ScaleUVsByCenter}
  n49 = Add(
   A=Multiply(
    A=ScalarParameter() {defaultValue=0.65, parameterName=DHMS_Brightness}
    B=TextureSampleParameter2D(UVs=@n48.UVs) {parameterName=DHMS, samplerType=SAMPLERTYPE_Masks, texture=/Game/Athena/Playlists/Barrier/Textures/Countdown_DHMS}.R
   )
   B=@n47.NumbersGreyscale
  )
  n50 = LinearInterpolate(Alpha=ComponentMask(None=TextureCoordinate()) {g=true}) {constA=0.6}
  n51 = MaterialFunctionCall(
   UVs=TextureCoordinate()
   Texture Scale=AppendVector(
    A=ScalarParameter() {defaultValue=0.7524, group=Numbers, parameterName=u_PostImagePaddingScalea}
    B=ScalarParameter() {defaultValue=1.1963, group=Numbers, parameterName=v_PostImagePaddingScalea}
   )
  ) {materialFunction=&ScaleUVsByCenter}
  n52 = Saturate(None=LinearInterpolate(
    A=Multiply(A=@n49, B=@n50)
    B=TextureSampleParameter2D(UVs=@n51.UVs) {parameterName=PostCountdownImage, texture=&T_BlacksRGB}.R
    Alpha=CollectionParameter() {collection=&MPC_Countdown, parameterName=ShowCountdownPhase3}
   ))
  n53 = LinearInterpolate(A=Multiply(A=ScalarParameter() {defaultValue=0.5, parameterName=BackgroundMult}, B=@n42), Alpha=@n52)
  n54 = Multiply(A=@n42) {constB=0.15}
  n55 = MaterialFunctionCall(In=@n32, Contrast=ScalarParameter() {defaultValue=0.2, group=Noise, parameterName=NoiseContrast}) {materialFunction=&CheapContrast}
  n56 = Multiply(A=@n55.Result) {constB=2}
  n57 = Subtract(A=ScalarParameter() {defaultValue=0.3, parameterName=MinScanlineOpacity}, B=@n56)
  n58 = Saturate(None=@n55.Result)
  n59 = Multiply(A=ScalarParameter() {defaultValue=200, parameterName=ScanlineTile}, B=LinearInterpolate(Alpha=Ceil(None=@n58)) {constA=1, constB=1.25})
  n60 = ScalarParameter() {defaultValue=30000, parameterName=ScaleDistance}
  n61 = Divide(A=Distance(A=CameraPositionWS(), B=ActorPositionWS()), B=@n60) {constB=1000}
  n62 = Divide(A=ComponentMask(None=WorldPosition().XYZ) {b=true}, B=LinearInterpolate(B=@n60, Alpha=Saturate(None=@n61)) {constA=50, constB=100}) {constB=20}
  n63 = ConstantBiasScale(None=Sine(None=Multiply(A=@n59, B=@n62) {constA=1}))
  n64 = Saturate(None=LinearInterpolate(A=Clamp(None=@n57) {minDefault=-0.25}, B=Add(A=@n56), Alpha=Power(Base=@n63) {constExponent=1.5}) {constA=0.5})
  n65 = ScalarParameter() {defaultValue=1.25, parameterName=OpacityMult}
  Multiply(A=Multiply(A=Add(A=@n53, B=@n54), B=Multiply(A=@n64, B=@n65)), B=@n36)
  n66 = CurveAtlasRowParameter(CurveTime=Saturate(None=@n53)) {
   atlas=&AthenaCurveAtlas01
   curve=/Game/Athena/LinearColorCurves/Curve_Galileo_Hologram_Pink
   defaultValue=290
   parameterName=ColorCurve
  }
  Multiply(
   A=LinearInterpolate(
    A=ScalarParameter() {defaultValue=3, parameterName=EmissiveMult}
    B=VectorParameter() {parameterName=UIBrightness, defaultValue=<rgba 13.9919 0 50 1>}.RGB
    Alpha=@n52
   )
   B=@n66
  ) {constB=2}
 ]
}
